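; Player gameplay ability that lets a player float while holding a Nug, found as the held object attached to the pawn.
; After a server net-sync, it sends a float-start event to the Nug, applies glide and quest-tracking effects, and launches the player forward and up once falling, unless tagged Launched.
; Returning to walking sends float-end, applies a landing effect and ends the ability, whose cleanup removes the glide and quest effects.
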(event Ability|EventActivateAbilityFromEvent
  (params (EventData :type Gameplay%20Event%20Data%20Structure%20%28by%20ref%29))
  (bind _owning_player_pawn (Variables|Default|GetOwning%20Player%20Pawn))
  (|CallParentFunction EventData)
  (CallFunction|SetRefs)
  ;[1 1 1] Apply floating movement and send event for Nug montage when we start falling
  (Ability|Tasks|WaitNetSync "OnlyServerWait"
    (:OnSync
      (Ability|SendGameplayEventtoActor (Variables|Default|GetNugRef) (Variables|Default|GetFloatStart))
      ;[1 1 1] Apply low g and give a slight launch
      (bind _returnvalue (GameplayEffects|ApplyGameplayEffectToSelf :self (Ability|GetAbilitySystemComponent (Variables|Default|GetOwning%20Player%20Pawn)) :GameplayEffectClass (Variables|Default|GetGE_PlayerGlide) :Level 0.0))
      (Variables|Default|SetGE_Handle_Glide _returnvalue)
      (if (not (GameplayTags|HasMatchingGameplayTag (Ability|GetAbilitySystemComponent _owning_player_pawn) "(TagName=\"Gameplay.Action.Player.Launched\")"))
        (CallFunction|DoCharacterLaunchwhenFalling _owning_player_pawn))
      ;[1 1 1] Add GE to the player for Quest Tracking
      (bind _returnvalue_2 (GameplayEffects|ApplyGameplayEffectToSelf :self (Ability|GetAbilitySystemComponent (Variables|Default|GetOwning%20Player%20Pawn)) :GameplayEffectClass "/Irwin/AI/Prey/Nug/Abilities/PlayerAbilities/GE_Player_Irwin_Quest_NugFloating_Active.GE_Player_Irwin_Quest_NugFloating_Active_C" :Level 0.0))
      (Variables|Default|SetGE_Handle_Quest _returnvalue_2)
      (Ability|Tasks|WaitMovementModeChange "MOVE_Walking"
        (:OnChange
          (Ability|SendGameplayEventtoActor (Variables|Default|GetNugRef) (Variables|Default|GetFloatEnd))
          ;[1 1 1] Remove Quest Tracking GE
          (bind _owning_player_pawn_2 (Variables|Default|GetOwning%20Player%20Pawn)
            (:then
              (GameplayEffects|RemoveActiveGameplayEffect (Ability|GetAbilitySystemComponent _owning_player_pawn_2) (Variables|Default|GetGE_Handle_Quest) -1)
              (CallFunction|OnLand))
            (:else)))))))

(event Ability|EventOnEndAbility
  (params (bWasCancelled :type Boolean))
  (|CallParentFunction bWasCancelled)
  (Ability|EndTaskbyInstanceName (Variables|Default|GetFloatTask))
  (bind _nugref (Variables|Default|GetNugRef)
    (:then
      (Ability|SendGameplayEventtoActor _nugref (Variables|Default|GetFloatEnd)))
    (:else
      _))
  ;[1 1 1] Remove Quest Tracking GE
  (bind _owning_player_pawn (Variables|Default|GetOwning%20Player%20Pawn)
    (:then
      (Utilities|IsValid (Ability|GetAbilitySystemComponent _owning_player_pawn)
        (:Is%20Valid
          (GameplayEffects|RemoveActiveGameplayEffect (Ability|GetAbilitySystemComponent _owning_player_pawn) (Variables|Default|GetGE_Handle_Quest) -1)
          ;[1 1 1] Remove Glide GE
          (GameplayEffects|RemoveActiveGameplayEffect (Ability|GetAbilitySystemComponent _owning_player_pawn) (Variables|Default|GetGE_Handle_Glide) -1))))
    (:else)))

(event Custom|On%20Land
  (bind _owning_player_pawn (Variables|Default|GetOwning%20Player%20Pawn)
    (:then
      (GameplayEffects|ApplyGameplayEffectToSelf :self (Ability|GetAbilitySystemComponent _owning_player_pawn) :GameplayEffectClass (Variables|Default|GetGE_Land_Tag_GCN) :Level 0.0))
    (:else
      _))
  (Ability|EndAbility))

(event Custom|DoDelayedCharacterLaunch
  (params (InCharacter :type Character%20Object%20Reference))
  (Ability|Tasks|WaitMovementModeChange "MOVE_Falling"
    (:OnChange
      (CallFunction|DoCharacterLaunchImmediately InCharacter))))

(fn DoCharacterLaunchImmediately
  (params (InCharacter :type Character%20Object%20Reference))
  (bind (_returnvalue_x _returnvalue_y) (Math|Vector|BreakVector (* (Transformation|GetActorForwardVector (Variables|Default|GetOwning%20Player%20Pawn)) (ScalableFloat|GetScalableFloatMultiplier (Variables|Default|GetFloat_LaunchXY)))))
  (Character|LaunchCharacter InCharacter (Math|Vector|MakeVector _returnvalue_x _returnvalue_y (ScalableFloat|GetScalableFloatMultiplier (Variables|Default|GetFloat_LaunchZ))) false false))

(fn DoCharacterLaunchWhenFalling
  (params (InCharacter :type Character%20Object%20Reference))
  (if (Utilities|Enum|Equal(Enum) (Variables|/Script/Engine.CharacterMovementComponent|GetMovementMode (Variables|/Script/Engine.Character|GetCharacterMovement InCharacter)) "MOVE_Falling")
    (CallFunction|DoCharacterLaunchImmediately InCharacter)
    (else
      (CallFunction|DoDelayedCharacterLaunch InCharacter))))

(fn SetRefs
  (bind _returnvalue (Ability|GetAvatarActorfromActorInfo))
  (bind (_asfort_player_pawn _bsuccess) (Utilities|Casting|CastToFortPlayerPawn _returnvalue))
  (Variables|Default|SetOwning%20Player%20Pawn _asfort_player_pawn)
  (if _bsuccess
    (bind _returnvalue_2 (Fortnite|HeldObjects|GetHeldObjectAttachedtoPlayer _asfort_player_pawn))
    (Utilities|IsValid _returnvalue_2
      (:Is%20Valid
        (Variables|Default|SetNugRef _returnvalue_2))
      (:Is%20Not%20Valid
        _)))
  (Ability|EndAbility))
